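; Gameplay ability for Ranked Ch4S4 that grants the player a random Urgent transient quest in playlists tagged Habanero.
; It caches the player controller to survive pawn swaps, then listens to the stateful game-phase event for Aircraft or SafeZones.
; It grants two seconds after the player exits the Battle Bus, or after load if already out.
; The random pick draws from quests whose RankedCh4S4 curve table row is above zero.
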(event Ability|EventActivateAbility
  ;[0 0 0] Wait until the Aircraft phase has started, or the Zone phase if the player joined the match late
  (|CallParentFunction)
  ;;[1 1 1] Ability can activate multiple times (when jumping from Aircraft, or getting rebooted)
  (Utilities|FlowControl|DoOnce false
    (:Completed
      ;;[1 1 1] Ensure we have our FortPlayerControllerAthena set before moving forward (do this first so we can operate on the controller moving forward, making us resilient to pawn swaps)
      (bind _asfort_player_controller_athena (Utilities|Casting|CastToFortPlayerControllerAthena (Pawn|GetController (Variables|GAVars|Data|GetOwningPlayerPawn)))
        (:then
          (Variables|Default|SetOwningPlayerController _asfort_player_controller_athena)
          (bind (_gamestate _playlist _playlistcontexttags) (Class|FortAsyncActionCurrentPlaylistReady|PlaylistDataReadyAsync)
            (:OnReady
              (if (GameplayTags|HasTag _playlistcontexttags "(TagName=\"Athena.Playlist.Habanero\")" true)
                (bind _output_get (Variables|Default|SetAthenaGameState (Fortnite|Athena|GameState|GetGameStateAthenafromContext self)))
                (bind _returnvalue (GameplayEventRouter|GetEventRouter "/Script/FortniteGame.FortGameState" _output_get))
                (bind (_asynctaskproxy _payload) (GameplayEventRouter|StartListeningToStatefulEvent _returnvalue "/Script/FortniteGame.FortGamePhaseUpdatedEvent")
                  (:then
                    ;;[1 1 1] NOTE: This may get set after having run OnGamePhaseChanged (if SavedEventStateExists runs)
                    (Variables|Default|SetGamePhaseUpdatedAsyncTask _asynctaskproxy))
                  (:OnEventReceived
                    _)
                  (:OnSavedEventStateExists
                    _))
                (CallFunction|OnGamePhaseChanged (Utilities|Struct|BreakFortGamePhaseUpdatedEvent _payload))))))
        (:CastFailed
          (Pawn|BindEventtoReceiveControllerChangedDelegate (Variables|GAVars|Data|GetOwningPlayerPawn) (delegate Custom|OnControllerChanged)))))))

(event Custom|OnControllerChanged
  (params (Pawn :type Pawn%20Object%20Reference) (OldController :type Controller%20Object%20Reference) (NewController :type Controller%20Object%20Reference))
  ;[0 0 0] Wait until the Aircraft phase has started, or the Zone phase if the player joined the match late
  ;;[1 1 1] Ensure we have our FortPlayerControllerAthena set before moving forward (do this first so we can operate on the controller moving forward, making us resilient to pawn swaps)
  (bind _asfort_player_controller_athena (Utilities|Casting|CastToFortPlayerControllerAthena NewController))
  (Variables|Default|SetOwningPlayerController _asfort_player_controller_athena)
  (Pawn|UnbindEventfromReceiveControllerChangedDelegate (Variables|GAVars|Data|GetOwningPlayerPawn) (delegate Custom|OnControllerChanged))
  (bind (_gamestate _playlist _playlistcontexttags) (Class|FortAsyncActionCurrentPlaylistReady|PlaylistDataReadyAsync)
    (:OnReady
      (if (GameplayTags|HasTag _playlistcontexttags "(TagName=\"Athena.Playlist.Habanero\")" true)
        (bind _output_get (Variables|Default|SetAthenaGameState (Fortnite|Athena|GameState|GetGameStateAthenafromContext self)))
        (bind _returnvalue (GameplayEventRouter|GetEventRouter "/Script/FortniteGame.FortGameState" _output_get))
        (bind (_asynctaskproxy _payload) (GameplayEventRouter|StartListeningToStatefulEvent _returnvalue "/Script/FortniteGame.FortGamePhaseUpdatedEvent")
          (:then
            ;;[1 1 1] NOTE: This may get set after having run OnGamePhaseChanged (if SavedEventStateExists runs)
            (Variables|Default|SetGamePhaseUpdatedAsyncTask _asynctaskproxy))
          (:OnEventReceived
            _)
          (:OnSavedEventStateExists
            _))
        (CallFunction|OnGamePhaseChanged (Utilities|Struct|BreakFortGamePhaseUpdatedEvent _payload))))))

(event Custom|OnGamePhaseChanged
  (params (GamePhase :type EAthenaGamePhase%20Enum))
  ;[0 0 0] Wait until the Aircraft phase has started, or the Zone phase if the player joined the match late
  (if (CallFunction|HasMatchStarted GamePhase)
    (bind _gamephaseupdatedasynctask (Variables|Default|GetGamePhaseUpdatedAsyncTask)
      (:then
        (GameplayEventRouter|StopListeningtoStatefulEvent _gamephaseupdatedasynctask))
      (:else
        _))
    (CallFunction|StartGrantUrgentQuestProcess)))

(event Custom|StartGrantUrgentQuestProcess
  ;[0 0 0] Grant the Urgent quest when the player jumps from the Bus, or immediately if they load after having jumped
  ;;[1 1 1] Protect against StopListening running too early and not cleaning the listener properly the first time
  (Utilities|FlowControl|DoOnce false
    (:Completed
      (if (Aircraft|IsinAircraft (Variables|Default|GetOwningPlayerController) :!class /Script/FortniteGame.FortPlayerController)
        (bind _returnvalue (Class|FortPlayerController|GetAircraftComponent (Variables|Default|GetOwningPlayerController)))
        (Utilities|IsValid _returnvalue
          (:Is%20Valid
            ;;[1 1 1] Grant the quest after the player leaves the Aircraft
            (EventDispatchers|AssignOnExitAircraftEvent _returnvalue (delegate Custom|OnExitAircraftEvent_Event)))
          (:Is%20Not%20Valid
            _))
        (else
          _))
      ;;[1 1 1] Give a little breahing room before granting the quest, so it doesn't add more noise on top of other visual cues (either just jumped from the Bus, or they are loading late into the match).
      ;; Delay isn't needed for gameplay and could be removed
      (Utilities|FlowControl|Delay 2.0)
      (CallFunction|CheckEnabledQuests)
      (bind _output_get (Variables|Default|SetUrgentQs2QuestToGrant (Utilities|Array|RandomArrayItem (Variables|Default|GetUrgentQs2QuestToGrant_Array_Enabled))))
      (Class|FortControllerComponentTransientQuests|GrantTransientQuest (Class|FortPlayerControllerAthena|GetTransientQuestsComponent (Variables|Default|GetOwningPlayerController)) _output_get 0))))

(event Custom|OnExitAircraftEvent_Event
  (params (Controller :type Controller%20Object%20Reference))
  ;[0 0 0] Grant the Urgent quest when the player jumps from the Bus, or immediately if they load after having jumped
  ;;[1 1 1] Give a little breahing room before granting the quest, so it doesn't add more noise on top of other visual cues (either just jumped from the Bus, or they are loading late into the match).
  ;; Delay isn't needed for gameplay and could be removed
  (Utilities|FlowControl|Delay 2.0)
  (CallFunction|CheckEnabledQuests)
  (bind _output_get (Variables|Default|SetUrgentQs2QuestToGrant (Utilities|Array|RandomArrayItem (Variables|Default|GetUrgentQs2QuestToGrant_Array_Enabled))))
  (Class|FortControllerComponentTransientQuests|GrantTransientQuest (Class|FortPlayerControllerAthena|GetTransientQuestsComponent (Variables|Default|GetOwningPlayerController)) _output_get 0))

(event Ability|EventOnEndAbility
  (params (bWasCancelled :type Boolean))
  (|CallParentFunction bWasCancelled)
  (bind _gamephaseupdatedasynctask (Variables|Default|GetGamePhaseUpdatedAsyncTask)
    (:then
      (GameplayEventRouter|StopListeningtoStatefulEvent _gamephaseupdatedasynctask))
    (:else)))

(fn CheckEnabledQuests
  (for _array_element (Variables|Default|GetUrgentQs2QuestToGrant_Array)
    (Variables|Default|SetArray%20Element_0 _array_element)
    (bind _outxy (DataTable|EvaluateCurveTableRow "/BattlePassS26_Quests/Items/QuestItems/RankedCh4S4/RankedCh4S4.RankedCh4S4" (Utilities|String|StringToName (Item|GetPersistentName _array_element)))
      (:RowFound
        (if (> _outxy 0)
          (Utilities|Array|AddUnique (Variables|Default|GetUrgentQs2QuestToGrant_Array_Enabled) (Variables|Default|GetArray%20Element_0))))))
  (return))

(fn HasMatchStarted
  (params (GamePhase :type EAthenaGamePhase%20Enum))
  (return (or (Utilities|Enum|Equal(Enum) GamePhase "Aircraft") (Utilities|Enum|Equal(Enum) GamePhase "SafeZones"))))
